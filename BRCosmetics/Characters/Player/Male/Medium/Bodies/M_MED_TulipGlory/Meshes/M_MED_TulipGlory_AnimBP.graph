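# Animation Blueprint for the TulipGlory male body cosmetic, which follows the attached parent mesh and adjusts extra spine bones.
# A copied pose runs through a Control Rig, then weighted multi-target constraints drive the base_spine bones.
# Fixed bone-space offsets then adjust the left and right dyn_spine bones.
# Event graph logic detects backblings, capes and tails to set tail visibility and TailAlpha, but the AnimGraph reads none of those variables.
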AnimBlueprint /BRCosmetics/Characters/Player/Male/Medium/Bodies/M_MED_TulipGlory/Meshes/M_MED_TulipGlory_AnimBP
target_skeleton=&/M_MED_TulipGlory_Skeleton
graphs [
 AnimGraph {nodes [
   n1 = "Transform (Modify) Bone - Bone: dyn_spine_01_l"!(
    ComponentPose=Local To Component(LocalPose=Use cached pose 'CRCache_Constraints'())
    bAlphaBoolEnabled:bool=True
    Translation="0,1.5,-1.2"
    Rotation="-14,0,-25"
    Scale="1,1,0.7"
   ) -> (Pose) {
    alphaInputType=Bool
    rotationMode=BMM_Additive
    rotationSpace=BCS_BoneSpace
    scaleMode=BMM_Additive
    scaleSpace=BCS_BoneSpace
    alphaBoolBlend {blendOption=Linear}
    boneToModify {boneName=dyn_spine_01_l}
    rotation {pitch=-14, roll=-25}
    scale {z=0.7}
    translation {y=1.5, z=-1.2}
   }
   n2 = "Transform (Modify) Bone - Bone: dyn_spine_01_r"!(ComponentPose=@n1, bAlphaBoolEnabled:bool=True, Translation="0,1.5,-1.2", Rotation="-14,0,24", Scale="1,1,0.7") -> (Pose) {
    alphaInputType=Bool
    rotationMode=BMM_Additive
    rotationSpace=BCS_BoneSpace
    scaleMode=BMM_Additive
    scaleSpace=BCS_BoneSpace
    translationMode=BMM_Additive
    translationSpace=BCS_BoneSpace
    alphaBoolBlend {blendOption=Linear}
    boneToModify {boneName=dyn_spine_01_r}
    rotation {pitch=-14, roll=24}
    scale {z=0.7}
    translation {y=1.5, z=-1.2}
   }
   n3 = "Transform (Modify) Bone - Bone: dyn_spine_02_l"!(ComponentPose=@n2, bAlphaBoolEnabled:bool=True, Translation="0,0,0", Rotation="-14,0,-9", Scale="1,1,0.7") -> (Pose) {
    alphaInputType=Bool
    rotationMode=BMM_Additive
    rotationSpace=BCS_BoneSpace
    scaleMode=BMM_Additive
    scaleSpace=BCS_BoneSpace
    alphaBoolBlend {blendOption=Linear}
    boneToModify {boneName=dyn_spine_02_l}
    rotation {pitch=-14, roll=-9}
    scale {z=0.7}
   }
   n4 = "Transform (Modify) Bone - Bone: dyn_spine_02_r"!(ComponentPose=@n3, bAlphaBoolEnabled:bool=True, Translation="0,0,0", Rotation="-14,0,9", Scale="1,1,0.7") -> (Pose) {
    alphaInputType=Bool
    rotationMode=BMM_Additive
    rotationSpace=BCS_BoneSpace
    scaleMode=BMM_Additive
    scaleSpace=BCS_BoneSpace
    alphaBoolBlend {blendOption=Linear}
    boneToModify {boneName=dyn_spine_02_r}
    rotation {pitch=-14, roll=9}
    scale {z=0.7}
   }
   Output Pose(Result=Component To Local(ComponentPose=@n4)) {layerGroup=None}
   n5 = Control Rig(Source=Copy Pose From Mesh() {bUseAttachedParent=true}) {
    alphaBoolBlend {blendOption=Linear}
    controlRigAssetReference {blueprintRigClass=&/M_MED_TulipGlory_CtrlRig.M_MED_TulipGlory_CtrlRig_C}
   }
   Save cached pose 'CRCache_InputPose'(Pose=@n5) {cacheName=CRCache_InputPose}
   n6 = "Constraint - base_spine_02_r (PPPPP)"!(
    ComponentPose=Local To Component(LocalPose=Use cached pose 'CRCache_InputPose'())
    Alpha=1
    ConstraintWeights_0=0.5206
    ConstraintWeights_1:float=0.1817
    ConstraintWeights_2:float=0.1461
    ConstraintWeights_3:float=0.0842
    ConstraintWeights_4:float=0.0674
   ) -> (Pose) {
    alphaBoolBlend {blendOption=Linear}
    boneToModify {boneName=base_spine_02_r}
    constraintSetup [
     {targetBone {boneName=spine_03}}
     {targetBone {boneName=spine_02}}
     {targetBone {boneName=spine_01}}
     {targetBone {boneName=spine_04}}
     {targetBone {boneName=pelvis}}
    ]
    constraintWeights [1,1,1,1,1]
   }
   n7 = "Constraint - base_spine_02_l (PPPPP)"!(
    ComponentPose=@n6
    Alpha=1
    ConstraintWeights_0=0.5206
    ConstraintWeights_1:float=0.1817
    ConstraintWeights_2:float=0.1461
    ConstraintWeights_3:float=0.0842
    ConstraintWeights_4:float=0.0674
   ) -> (Pose) {
    alphaBoolBlend {blendOption=Linear}
    boneToModify {boneName=base_spine_02_l}
    constraintSetup [
     {targetBone {boneName=spine_03}}
     {targetBone {boneName=spine_02}}
     {targetBone {boneName=spine_01}}
     {targetBone {boneName=spine_04}}
     {targetBone {boneName=pelvis}}
    ]
    constraintWeights [1,1,1,1,1]
   }
   n8 = "Constraint - base_spine_01_r (PPPPP)"!(
    ComponentPose=@n7
    Alpha=1
    ConstraintWeights_0=0.6348
    ConstraintWeights_1:float=0.1451
    ConstraintWeights_2:float=0.0987
    ConstraintWeights_3:float=0.0795
    ConstraintWeights_4:float=0.0419
   ) -> (Pose) {
    alphaBoolBlend {blendOption=Linear}
    boneToModify {boneName=base_spine_01_r}
    constraintSetup [
     {targetBone {boneName=spine_03}}
     {targetBone {boneName=spine_04}}
     {targetBone {boneName=spine_02}}
     {targetBone {boneName=spine_01}}
     {targetBone {boneName=clavicle_l}}
    ]
    constraintWeights [1,1,1,1,1]
   }
   n9 = "Constraint - base_spine_01_l (PPPPP)"!(
    ComponentPose=@n8
    Alpha=1
    ConstraintWeights_0=0.6348
    ConstraintWeights_1:float=0.1451
    ConstraintWeights_2:float=0.0987
    ConstraintWeights_3:float=0.0795
    ConstraintWeights_4:float=0.0419
   ) -> (Pose) {
    alphaBoolBlend {blendOption=Linear}
    boneToModify {boneName=base_spine_01_l}
    constraintSetup [
     {targetBone {boneName=spine_03}}
     {targetBone {boneName=spine_04}}
     {targetBone {boneName=spine_02}}
     {targetBone {boneName=spine_01}}
     {targetBone {boneName=clavicle_r}}
    ]
    constraintWeights [1,1,1,1,1]
   }
   Save cached pose 'CRCache_Constraints'(Pose=Component To Local(ComponentPose=@n9)) {cacheName=CRCache_Constraints}
  ]}
 EventGraph {nodes [
   Event BlueprintUpdateAnimation()
   n1 = CheckForBackbling()
   > Set isWearingPackBool(isWearingPackBool:bool=OR Boolean!(A=@n1, B=@n1.bIsBackblingACape, C:bool=@n1.bIsBackblingLong) -> (ReturnValue)) -> (Output_Get:bool)
   n2 = Event OnCharacterCustomizationLoaded()
   > Set BodyMesh(BodyMesh=@n2.CustomBaseMesh)
   > n3 = Cast To PlayerPawn_Athena(Object=TryGetPawnOwner())
   > n4 = Set PawnOwner(PawnOwner:Player Pawn Athena Object Reference=@n3) -> (Output_Get:Player Pawn Athena Object Reference)
   > Set TailMesh(TailMesh:Skeletal Mesh Component Object Reference=GetSkeletalMeshForPartType(self=@n4, PartType=MiscOrTail)) -> (Output_Get:Skeletal Mesh Component Object Reference)
   n5 = CheckForBackbling()
   > Set isWearingBackbling(isWearingBackbling=@n5)
   > Set isWearingCape(isWearingCape=@n5.bIsBackblingLong)
   > Set characterHasTail(characterHasTail:bool=CheckForTail!(self:Custom Character Part Anim Instance Object Reference) -> (bHasTail:bool)) -> (Output_Get:bool)
   > Branch(Condition=Get characterHasTail() -> (characterHasTail:bool))
   > n6 = Branch(Condition=Get isWearingCape() -> (isWearingCape:bool))
   n7 = Get TailMesh() -> (TailMesh:Skeletal Mesh Component Object Reference)
   > SetVisibility!(execute, self:Scene Component Object Reference=@n7, bNewVisibility:bool=false, bPropagateToChildren:bool=false) -> (then)
   > Set cape and tail combo(cape and tail combo:bool=false) -> (Output_Get:bool)
   > n8 = Set cape and tail(cape and tail:bool=false) -> (Output_Get:bool)
   > Set TailAlpha(TailAlpha:double="To Float (Boolean)"(InBool=@n8)) -> (Output_Get:double)
   SetVisibility(execute=@n6.else, self=@n7, bNewVisibility=true, bPropagateToChildren=false)
   > n9 = Set cape and tail(cape and tail=true)
   > n10 = Branch(Condition=Get bIsInFrontEnd())
   > n11 = Set Cape and Tail Combo Frontend(Cape and Tail Combo Frontend:bool=true) -> (Output_Get:bool)
   > Set TailAlpha(TailAlpha="To Float (Boolean)"(InBool=@n11))
   n12 = Set cape and tail combo(execute=@n10.else, cape and tail combo=true)
   > Set TailAlpha(TailAlpha="To Float (Boolean)"(InBool=@n12))
   > Set TailAlpha(TailAlpha="To Float (Boolean)"(InBool=@n9))
  ]}
]
variables [
 BodyMesh {type=SkeletalMeshComponent}
 cape and tail {type=boolean}
 cape and tail combo {type=boolean}
 Cape and Tail Combo Frontend {type=boolean}
 characterHasTail {type=boolean}
 isWearingBackbling {type=boolean}
 isWearingCape {type=boolean}
 isWearingPackBool {type=boolean}
 PawnOwner {type=PlayerPawn_Athena_C}
 TailAlpha {type=number}
 TailMesh {type=SkeletalMeshComponent}
]
