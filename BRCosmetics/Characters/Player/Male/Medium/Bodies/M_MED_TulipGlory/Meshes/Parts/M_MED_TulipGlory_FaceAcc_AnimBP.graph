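# Post-copy animation blueprint for the TulipGlory medium male character's face accessory, following the attached parent mesh.
# It additively counter-rotates the head against its neck_02-relative delta while skydiving, scaled per axis.
# A half-strength jaw constraint on the inner mouth, an upper-teeth offset and a Control Rig then refine the pose.
# In the front end, the Frontend_Idle_Bypass curve blends DefaultSlot onto mouth and face-attach bone branches.
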AnimBlueprint /BRCosmetics/Characters/Player/Male/Medium/Bodies/M_MED_TulipGlory/Meshes/Parts/M_MED_TulipGlory_FaceAcc_AnimBP
target_skeleton=&/M_MED_TulipGlory_FaceAcc_Skeleton
graphs [
 AnimGraph {nodes [
   n1 = Layered blend per bone(
    BasePose=Use cached pose 'Inicial Pose Cache'()
    BlendPoses_0=Slot 'DefaultSlot'()
    BlendWeights_0=Get Custom Idle is Playing Float() -> (Custom Idle is Playing Float:double)
   ) {layerSetup [{branchFilters [
       {boneName=dyn_innerMouth_mstr}
       {boneName=C_lipUp_mid}
       {boneName=faceAttach}
       {boneName=L_lipUp_side}
       {boneName=R_lipUp_side}
       {boneName=teeth_upper}
      ]}]}
   Output Pose(Result=@n1) {layerGroup=None}
   n2 = "Transform (Modify) Bone - Bone: head"(
    ComponentPose=Local To Component(LocalPose=Copy Pose From Mesh() {bUseAttachedParent=true})
    Alpha=1
    Translation="0,0,0"
    Rotation=Get head rot() -> (head rot:Rotator)
    Scale="1,1,1"
   ) {rotationMode=BMM_Additive, rotationSpace=BCS_BoneSpace, alphaBoolBlend {blendOption=Linear}, boneToModify {boneName=head}}
   n3 = "Constraint - dyn_innerMouth_mstr (P)"(ComponentPose=@n2, Alpha=0.5, ConstraintWeights_0=1) {
    alphaBoolBlend {blendOption=Linear}
    boneToModify {boneName=dyn_innerMouth_mstr}
    constraintSetup [{targetBone {boneName=C_jaw}}]
    constraintWeights [1]
   }
   n4 = "Transform (Modify) Bone - Bone: teeth_upper"(ComponentPose=@n3, Alpha=1, Translation="0,-0.1,0", Rotation="0,0,0", Scale="1,1,1") {
    translationMode=BMM_Additive
    translationSpace=BCS_BoneSpace
    alphaBoolBlend {blendOption=Linear}
    boneToModify {boneName=teeth_upper}
    translation {y=-0.1}
   }
   n5 = Control Rig(Source=Component To Local(ComponentPose=@n4)) {
    alphaBoolBlend {blendOption=Linear}
    controlRigAssetReference {blueprintRigClass=&/M_MED_TulipGlory_FaceAcc_CtrlRig.M_MED_TulipGlory_FaceAcc_CtrlRig_C}
   }
   Save cached pose 'Inicial Pose Cache'(Pose=@n5) {cacheName=Inicial Pose Cache}
  ]}
 EventGraph {nodes [
   Event BlueprintUpdateAnimation()
   > n1 = Sequence!(execute) -> (then_0, then_1, then_2:Exec)
   > Branch(Condition=Get bIsInFrontEnd())
   > Set Custom Idle is Playing Float(Custom Idle is Playing Float:double=GetCurveValue(self=GetOwnerFrontEndAnimInstance(), CurveName=Frontend_Idle_Bypass)) -> (Output_Get:double)
   n2 = Branch(execute=@n1.then_1, Condition=Get bIsSkyDiving())
   > Set Z percentage(Z percentage:float=0.5) -> (Output_Get:float)
   > Set X percentage(X percentage:float=0.8) -> (Output_Get:float)
   Set Z percentage(execute=@n2.else, Z percentage=0)
   > Set X percentage(X percentage=0)
   n3 = GetOwnerSkeletalMeshComponent()
   n4 = BreakRotator(InRot=BreakTransform(InTransform=GetDeltaTransformFromRefPose(self=@n3, BoneName=head, BaseName=neck_02)).Rotation)
   n5 = Get Z percentage() -> (Z percentage:float)
   n6 = Get X percentage() -> (X percentage:float)
   n7 = Get Y percentage() -> (Y percentage:float)
   Set head rot(
    execute=@n1.then_2
    head rot:Rotator=MakeRotator(
     Roll=float * float!(A=@n4, B=-1, C:double=@n6) -> (ReturnValue)
     Pitch=float * float(A=@n4.Pitch, B=-1, C=@n7)
     Yaw=float * float(A=@n4.Yaw, B=-1, C=@n5)
    )
   ) -> (Output_Get:Rotator)
  ]}
]
variables [
 Custom Idle is Playing Float {type=number}
 head rot {type=Rotator}
 head rot_E {type=Rotator}
 X percentage {type=number, default=0.7}
 X percentage_E {type=number}
 Y percentage {type=number}
 Z percentage {type=number, default=0.9}
 Z percentage_E {type=number, default=0.4}
]
